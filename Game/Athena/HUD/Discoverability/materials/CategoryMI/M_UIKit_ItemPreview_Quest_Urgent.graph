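# Translucent UI material for an item preview, named for urgent quests, that shows an item texture with a vignette-style alpha mask.
# Offset, viewport-scaled UVs pass through an animatable centered zoom before sampling the Item texture.
# Item alpha is edge-clipped, with a RemoveItem fade and a channel-packed option that sums RGB instead.
# The vignette alpha blends a radial gradient toward a bottom fade, which is selected by default via BottomMask.
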Material /Game/Athena/HUD/Discoverability/materials/CategoryMI/M_UIKit_ItemPreview_Quest_Urgent
graph {
 properties {
  bUsedWithStaticMesh=false
  blendMode=BLEND_Translucent
  materialDomain=MD_UI
  pixelDepthOffsetMode=PDOM_Legacy
  usedShadingModels=Default Lit
 }
 nodes [
  Multiply(A=OneMinus(None=NamedRerouteUsage()), B=NamedRerouteUsage())
  n1 = NamedRerouteUsage()
  n2 = Multiply(
   A=AppendVector(A=NamedRerouteUsage(), B=NamedRerouteUsage())
   B=Custom() {code="return View.ViewportScaleUI;", description=ViewportScaleFactor, outputType=CMOT_Float1, inputs []}
  )
  n3 = ScalarParameter() {desc="(0 - 1) Animate scale transition 'ScaleFrom' - 'ScaleTo'", group=Item, parameterName=ScaleAnimation, sliderMax=1, sortPriority=4}
  n4 = MaterialFunctionCall(
   UVs=MaterialFunctionCall(UV=Add(A=@n1, B=@n2)) {materialFunction=&MF_UI_UVScaling}.ScaleY
   Texture Scale=LinearInterpolate(
    A=ScalarParameter() {defaultValue=1, desc=Start value for scale animation., group=Item, parameterName=ScaleFrom, sortPriority=2}
    B=ScalarParameter() {defaultValue=1.1, desc=End value for scale animation., group=Item, parameterName=ScaleTo, sortPriority=3}
    Alpha=@n3
   )
  ) {materialFunction=&ScaleUVsByCenter}
  n5 = TextureSampleParameter2D(UVs=@n4.UVs) {
   group=Item
   parameterName=Item
   samplerSource=SSM_Clamp_WorldGroupSettings
   sortPriority=1
   texture=/FortUILibrary/Systems/Items/Textures/T_UI_SoldierCommando_F
  }
  NamedRerouteDeclaration(None=@n5.RGB) {name=Item, nodeColor=<rgba 0 0 0 1>}
  n6 = ScalarParameter() {group=2 - Texture, parameterName=Texture_XOffset, sliderMax=100, sliderMin=-100}
  n7 = ComponentMask(None=NamedRerouteUsage()) {r=true}
  NamedRerouteDeclaration(None=Divide(A=Multiply(A=@n6) {constB=-1}, B=Max(A=@n7))) {name=Texture_XOffset, nodeColor=<rgba 0.0257 1 0 1>}
  n8 = MaterialFunctionCall!(UVs=@n4.UVs, Alpha=@n5.A) {materialFunction=/FortUILibrary/MaterialFunctions/Basics/MF_UI_ClipEdgePixel}
  n9 = ScalarParameter() {group=Item, parameterName=RemoveItem}
  n10 = Add(A=@n5.B, B=Add(A=@n5.R, B=@n5.G))
  n11 = ScalarParameter() {
   desc="(0 or 1) If enabled switch to work with channel-packed texture."
   group=ChannelPacked
   parameterName=UseChannelPacked
   sliderMax=1
   sortPriority=1
  }
  NamedRerouteDeclaration(None=LinearInterpolate(A=LinearInterpolate(A=@n8.Result, Alpha=@n9) {constB=0}, B=Saturate(None=@n10), Alpha=@n11)) {name=Item Alpha, nodeColor=<rgba 0 0 0 1>}
  n12 = MaterialFunctionCall() {materialFunction=&GetUserInterfaceUV}
  NamedRerouteDeclaration(None=@n12.Pixel Size) {name=PixelSize, nodeColor=<rgba 0.0262 1 0 1>}
  n13 = ScalarParameter() {group=2 - Texture, parameterName=Texture_YOffset}
  n14 = ComponentMask(None=NamedRerouteUsage()) {g=true}
  NamedRerouteDeclaration(None=Divide(A=Multiply(A=@n13) {constB=-1}, B=Max(A=@n14))) {name=Texture_YOffset, nodeColor=<rgba 0.0257 1 0 1>}
  NamedRerouteDeclaration(None=@n12.9-Slice UV) {name=UVs, nodeColor=<rgba 0.0262 1 0 1>}
  n15 = MaterialFunctionCall(
   CenterPosition=AppendVector(
    A=ScalarParameter() {defaultValue=0.5, group=Vignette, parameterName=VignetteCenter_X, sortPriority=10}
    B=ScalarParameter() {group=Vignette, parameterName=VignetteCenter_Y, sortPriority=11}
   )
   Radius=ScalarParameter() {defaultValue=1.4, group=Vignette, parameterName=VignetteRadius, sortPriority=12}
   Density=ScalarParameter() {defaultValue=3.5, group=Vignette, parameterName=VignetteDensity, sortPriority=13}
  ) {materialFunction=&RadialGradientExponential}
  n16 = Saturate(None=Multiply(
    A=OneMinus(None=@n15.RadialGradientExponential)
    B=ScalarParameter() {defaultValue=2, group=Vignette, parameterName=VignetteStrength, sortPriority=6}
   ))
  n17 = ScalarParameter() {
   defaultValue=1
   desc="(0 or 1) If enabled applies vignette effect on the item texture."
   group=Vignette
   parameterName=UseVignette
   sliderMax=1
   sortPriority=0
  }
  n18 = ComponentMask(None=TextureCoordinate()) {g=true}
  n19 = ScalarParameter() {defaultValue=0.5, parameterName=FadeCutOff}
  n20 = Multiply(
   A=Saturate(None=Subtract(A=@n18, B=@n19))
   B=ScalarParameter() {defaultValue=2, group=Vignette, parameterName=VignetteStrength, sortPriority=6}
  )
  NamedRerouteDeclaration(None=LinearInterpolate(A=Multiply(A=@n16, B=@n17), B=Saturate(None=@n20), Alpha=ScalarParameter() {defaultValue=1, parameterName=BottomMask})) {name=Vignette Alpha, nodeColor=<rgba 0 0 0 1>}
  NamedRerouteUsage()
 ]
}
