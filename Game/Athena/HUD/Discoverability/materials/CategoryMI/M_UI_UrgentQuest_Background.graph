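# Translucent UI material for the urgent quest background in the HUD Discoverability category set, with rounded corners from per-corner radius parameters.
# A rotatable three-color linear gradient forms the base, composited with a scaled item texture and a rotated, soft corner color fade.
# The texture's alpha is multiplied by a constant zero, so ShowItemImage has no effect.
# The IsDebugDisplay static switch provides an alternative debug look, built partly from a screen blend with the Slate post buffer.
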Material /Game/Athena/HUD/Discoverability/materials/CategoryMI/M_UI_UrgentQuest_Background
graph {
 properties {
  bAutomaticallySetUsageInEditor=false
  bUsedWithStaticMesh=false
  blendMode=BLEND_Translucent
  materialDomain=MD_UI
  usedShadingModels=Default Lit
 }
 nodes [
  n1 = VectorParameter() {
   group=Layout
   parameterName=Corner Radius
   sortPriority=2
   channelNames {a=Bottom Right, b=Top Right, g=Bottom Left, r=Top Left}
   defaultValue=<rgba 10 10 10 10>
  }
  Multiply(
   A=Multiply(A=NamedRerouteUsage(), B=NamedRerouteUsage())
   B=MaterialFunctionCall(
    Corner Radius=AppendVector(A=@n1, B=@n1)
    Radius in Pixels=ScalarParameter() {defaultValue=1, group=Layout, parameterName=Radius In Pixels, sliderMax=1, sortPriority=3}
   ) {materialFunction=&MF_UI_CoreUI_RoundedMask}.Shape Mask
  )
  n2 = VectorParameter() {desc=LeftColor, group=1 - Background, parameterName=BackgroundColor, defaultValue=<rgba 0.007 0.1221 0.1329 1>}
  NamedRerouteDeclaration(None=@n2.RGB) {name=Background_Color, nodeColor=<rgba 1 0.9316 0.0156 1>}
  n3 = NamedRerouteUsage()
  n4 = NamedRerouteUsage()
  n5 = NamedRerouteUsage()
  n6 = ScalarParameter() {defaultValue=0.7, desc="GradientMidOffset\r\nGradientStart < GradientMid < GradientEnd", group=1 - Background, parameterName=2 GradientMid}
  n7 = ScalarParameter() {desc=From 0 to 1, group=1 - Background, parameterName=4 GradientRotation}
  n8 = ScalarParameter() {desc=Full gradient offset, group=1 - Background, parameterName=5 GradientOffset}
  n9 = MaterialFunctionCall(
   GradientStart=ScalarParameter() {desc="Gradient Start Offset \r\nGradientStart < GradientMid < GradientEnd", group=1 - Background, parameterName=1 GradientStart}
   GradientEnd=@n6
   RotationAngle=@n7
   GradientOffset=@n8
  ) {materialFunction=&MF_UI_LinearGradient}
  n10 = ScalarParameter() {defaultValue=1.6, desc="Gradient End Offset\r\nGradientStart < GradientMid < GradientEnd", group=1 - Background, parameterName=3 GradientEnd}
  n11 = MaterialFunctionCall(GradientStart=@n6, GradientEnd=@n10, RotationAngle=@n7, GradientOffset=@n8) {materialFunction=&MF_UI_LinearGradient}
  n12 = Multiply(A=Subtract(A=LinearInterpolate(Alpha=@n11.Result) {constA=0.5}) {constB=0.5}) {constB=2}
  n13 = Constant() {r=1}
  n14 = Constant() {r=1}
  n15 = Constant() {r=0.95}
  n16 = Constant() {r=1}
  NamedRerouteDeclaration(None=LinearInterpolate(
    B=MaterialFunctionCall!(
     Color A=@n3
     Color B=@n4
     Color C=@n5
     A to B Gradient=@n9.Result
     Color C Space=@n12
     "Display Color C (0 to 1)"=@n13
     "Saturation (-1 to 1)"=@n14
     Saturation Mask
     Saturation Mask Threshold=@n15
     "Color C Covers Saturation (0 to 1)"=@n16
    ) {materialFunction=/Game/Athena/HUD/Discoverability/materials/Functions/MF_UI_Gradient_3Colors_WithSaturation}.Colorful RGBA Gradient
    Alpha=NamedRerouteUsage()
   )) {name=Background_RGB, nodeColor=<rgba 0 0.2462 1 1>}
  NamedRerouteDeclaration(None=MaterialFunctionCall(
    UV=MaterialFunctionCall() {materialFunction=&GetUserInterfaceUV}.Normalized UV
    Corner Radius=Constant() {r=16}
    Radius in Pixels=Constant() {r=1}
   ) {materialFunction=&MF_UI_CoreUI_RoundedMask}.Shape Mask) {name=RoundedFill_Mask, nodeColor=<rgba 0 0.2462 1 1>}
  n17 = MaterialFunctionCall() {materialFunction=&GetUserInterfaceUV}
  NamedRerouteDeclaration(None=@n17.Pixel Size) {name=PixelSize, nodeColor=<rgba 0.0262 1 0 1>}
  n18 = TextureObjectParameter() {group=2 - Texture, parameterName=Texture, texture=/FortUILibrary/Systems/Items/Textures/T_UI_SoldierCommando_F}
  NamedRerouteDeclaration(None=TextureProperty(Tex=@n18)) {name=Texture_Size, nodeColor=<rgba 0.0257 1 0 1>}
  NamedRerouteDeclaration(None=@n18) {name=Texture_Object, nodeColor=<rgba 0 0.2461 1 1>}
  n19 = NamedRerouteDeclaration() {name=UVs_1, nodeColor=<rgba 0.0262 1 0 1>}
  n20 = NamedRerouteUsage()
  n21 = Constant() {r=1}
  n22 = TextureSample(
   UVs=NamedRerouteDeclaration(None=MaterialFunctionCall(UV=@n19, TextureSize=@n20, ScalePivotPoint=@n21) {materialFunction=&MF_UI_UVScaling}.ScaleY) {name=TextureUVs, nodeColor=<rgba 0 0.2462 1 1>}
   Tex=NamedRerouteUsage()
  ) {samplerSource=SSM_Clamp_WorldGroupSettings}
  n23 = Constant() {r=1}
  n24 = MaterialFunctionCall!(UV=NamedRerouteUsage(), Grid Size=@n23, Index=@n23) {materialFunction=/FortUILibrary/MaterialFunctions/Basics/MF_UI_IsolateUV}
  NamedRerouteDeclaration(None=Multiply(A=@n22.RGB, B=@n24.Result)) {name=Texture_RGB, nodeColor=<rgba 0 0.2461 1 1>}
  n25 = Multiply(A=ScalarParameter() {desc="1 = texture fully opaque\r\n0 = texture completely transparent", group=2 - Texture, parameterName=ShowItemImage}) {constB=0}
  n26 = Multiply(A=@n22.A, B=@n24.Result)
  NamedRerouteDeclaration(None=Multiply(A=Multiply(A=@n25, B=@n26), B=NamedRerouteUsage())) {name=Texture_Alpha, nodeColor=<rgba 0 0.2462 1 1>}
  n27 = ScalarParameter() {
   defaultValue=1.9
   desc="Gradient softness\r\n0 = hard, sharp edge \r\n1 = smooth and soft"
   group=CornerGradient
   parameterName=CornerGradient_SmoothStepMax
  }
  NamedRerouteDeclaration(None=@n27) {name=CornerGradient_SmoothStepMax, nodeColor=<rgba 0.0257 1 0 1>}
  n28 = ScalarParameter() {defaultValue=3, desc="Don't use this please\r\nBetter use CornerGradient_Size", group=Legacy, parameterName=CornerGradient_XPosition}
  NamedRerouteDeclaration(None=MaterialFunctionCall(
    X=@n28
    Y=ScalarParameter() {desc="Don't use this please\r\nBetter use CornerGradient_Size", group=Legacy, parameterName=CornerGradient_YPosition}
    Z=ScalarParameter() {defaultValue=4, group=CornerGradient, parameterName=CornerGradient_Size}
   ) {materialFunction=&MakeFloat3}.Result) {name=CornerGradient_AnchorPosition, nodeColor=<rgba 0 0.2462 1 1>}
  n29 = NamedRerouteUsage()
  n30 = Constant() {r=0.3}
  n31 = AppendVector(A=@n30, B=@n30)
  n32 = NamedRerouteUsage()
  n33 = MaterialFunctionCall(UV Channel=MaterialFunctionCall(
    UVs=MaterialFunctionCall!(Texture Scale=@n31, UVs, Offset XY Texturesize Z=@n32) {materialFunction=/Engine/Functions/Engine_MaterialFunctions01/Texturing/ScaleUVsAroundPoint}.UVs
    Rotation Center=Constant() {r=3}
    "Rotation Angle (0-1)"=Constant() {r=-0.1}
   ) {materialFunction=&CustomRotator}.Rotated Values) {materialFunction=&LinearGradient}
  NamedRerouteDeclaration(None=Multiply(A=SmoothStep(Max=@n29, Value=@n33.UGradient) {constMax=1.5}, B=NamedRerouteUsage())) {name=CornerGradient_Mask, nodeColor=<rgba 0 0.2461 1 1>}
  n34 = VectorParameter() {group=CornerGradient, parameterName=CornerGradient_ColorFade, defaultValue=<rgba 0.0482 0.1329 0.0194 1>}
  NamedRerouteDeclaration(None=@n34.RGB) {name=CornerGradient_RGB, nodeColor=<rgba 0 0.2461 1 1>}
  NamedRerouteDeclaration(None=ComponentMask(None=@n34.A) {a=true}) {name=CornerGradient_Alpha, nodeColor=<rgba 0 0.2461 1 1>}
  n35 = VectorParameter() {group=Debug, parameterName=BackgroundDefaultColor, defaultValue=<rgba 0.099 0.099 0.099 0.5>}
  NamedRerouteDeclaration(None=@n35.RGB) {name=Background_DebugColor, nodeColor=<rgba 1 0.9316 0.0156 1>}
  NamedRerouteDeclaration(None=@n35.A) {name=Background_DebugAlpha, nodeColor=<rgba 1 0.9316 0.0156 1>}
  n36 = MaterialFunctionCall!("UVs (Optional)") {materialFunction=/Engine/Functions/UserInterface/GetSlatePost0}
  n37 = MaterialFunctionCall!(Color=Max(A=Saturate(None=@n36.LinearRGB)) {constB=0.0001}, Contrast=ScalarParameter() {defaultValue=0.4, parameterName=EffectWeakness}) {materialFunction=/Engine/Functions/Engine_MaterialFunctions01/ImageAdjustment/Contrast_Preserve_Color}
  NamedRerouteDeclaration(None=MaterialFunctionCall!(Blend=Saturate(None=@n37.Result), Base=NamedRerouteUsage()) {materialFunction=/Engine/Functions/Engine_MaterialFunctions03/Blends/Blend_Screen}.Result) {name=Background_DebugRGB, nodeColor=<rgba 0 0.2461 1 1>}
  n38 = NamedRerouteUsage()
  n39 = NamedRerouteUsage()
  n40 = NamedRerouteUsage()
  NamedRerouteDeclaration(None=LinearInterpolate(A=LinearInterpolate(A=@n38, B=@n39, Alpha=@n40) {constAlpha=0}, B=NamedRerouteUsage(), Alpha=NamedRerouteUsage())) {name=CompositeBackground_RGB, nodeColor=<rgba 0 0.2462 1 1>}
  NamedRerouteDeclaration(None=@n17.9-Slice UV) {name=UVs, nodeColor=<rgba 0.0262 1 0 1>}
  n41 = VectorParameter() {desc=CenterColor, group=1 - Background, parameterName=ColorSecond, defaultValue=<rgba 0.1413 0.5149 0.0065 1>}
  NamedRerouteDeclaration(None=@n41.RGB) {name=Background_ColorSecond, nodeColor=<rgba 1 0.9316 0.0156 1>}
  n42 = VectorParameter() {desc=RightColor, group=1 - Background, parameterName=ColorCenter, defaultValue=<rgba 0.3095 1 0.0137 1>}
  NamedRerouteDeclaration(None=@n42.RGB) {name=Background_ColorCenter, nodeColor=<rgba 1 0.9316 0.0156 1>}
  n43 = ScalarParameter() {
   defaultValue=0.7
   desc="0 = completely transparent \r\n1 = fully opaque "
   group=1 - Background
   parameterName=Background Opacity
   sliderMax=1
  }
  NamedRerouteDeclaration(None=StaticSwitchParameter(True=NamedRerouteUsage(), False=@n43) {parameterName=IsDebugDisplay}) {name=Background_Opacity, nodeColor=<rgba 0 0.2461 1 1>}
  StaticSwitchParameter(True=NamedRerouteUsage(), False=NamedRerouteUsage()) {parameterName=IsDebugDisplay}
 ]
}
